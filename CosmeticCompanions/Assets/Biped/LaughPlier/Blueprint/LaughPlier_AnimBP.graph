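# LaughPlier_AnimBP animates the LaughPlier biped companion on the FrankenBruin skeleton, with a rolling-ball rotation on spine_02.
# An Idle/Run/Stop machine feeds an aim offset and additive turn blendspace, cached before a voice-envelope curve, slots and dead blending.
# spine_02 is reset in world space, then rotated by speed-accumulated roll and direction-smoothed tilt, updated in both UpdateBallRoll and the event graph.
# Stop outputs only StopLight; chooser, heavy stop and head-bone nodes are unconnected.
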AnimBlueprint /CosmeticCompanions/Assets/Biped/LaughPlier/Blueprint/LaughPlier_AnimBP
target_skeleton=/CosmeticCompanions/Assets/Biped/FrankenBruin/Meshes/FrankenBruin_Biped_Skeleton
graphs [
 AnimGraph {nodes [
   "Transform (Modify) Bone - Bone: head"!(
    ComponentPose
    Alpha=1
    Rotation_Roll:float=float * float(A=Get LookAtYaw() -> (LookAtYaw:float), B=-1) {is=title K2Node_PromotableOperator}
    Rotation_Pitch:float=0
    Rotation_Yaw:float=float * float(A=Get LookAtPitch() -> (LookAtPitch:float), B=-0.25)
   ) -> (Pose) {rotationMode=BMM_Additive, rotationSpace=BCS_BoneSpace, boneToModify {boneName=head}, rotation=<rot -0.017 -2.5 -0.395>}
   n1 = Modify Curve(SourcePose=Use cached pose 'PreVOPose'(), CurveValues_0=1, Alpha=Get AudioEnvelopeValue() -> (AudioEnvelopeValue:float)) {curveNames [VOActivity], curveValues [0]}
   n2 = Use cached pose 'PreVOPose'()
   n3 = Get bIsVOActive() -> (bIsVOActive:bool)
   n4 = Slot 'SequencerDefault'(Source=Slot 'PostVO'(Source=Blend Poses by bool(BlendPose_0=@n1, BlendPose_1=@n2, bActiveValue=@n3)) {slotName=PostVO}) {slotName=SequencerDefault}
   n5 = "Transform (Modify) Bone - Bone: spine_02"(
    ComponentPose=Local To Component(LocalPose=Dead Blending!(Source:Pose Link Structure=@n4) -> (Pose:Pose Link Structure) {is=title AnimGraphNode_DeadBlending, defaultBlendMode=Linear})
    Alpha=1
    Translation="0,0,0"
    Rotation="0,0,0"
    Scale="1,1,1"
   ) {rotationMode=BMM_Replace, rotationSpace=BCS_WorldSpace, boneToModify {boneName=spine_02}}
   n6 = Get BallRotation() -> (BallRotation:Rotator)
   Output Pose(Result=Component To Local(ComponentPose="Transform (Modify) Bone - Bone: spine_02"(ComponentPose=@n5, Alpha=1, Translation="0,0,0", Rotation=@n6, Scale="1,1,1") {rotationMode=BMM_Additive, boneToModify {boneName=spine_02}}))
   n7 = Get CompanionDirection() -> (CompanionDirection:float)
   AnimGraphNode_StateMachine_0 = Locomotion() {is=title AnimGraphNode_StateMachine}
   n8 = Get LookAtYaw()
   n9 = Get LookAtPitch()
   Save cached pose 'PreVOPose'(Pose=Apply Additive(
     Base=AimOffset Player 'FrankenBruin_AO'!(
      X:float=@n8
      Y:float=@n9
      BlendSpace:Blend Space Object Reference=/CosmeticCompanions/Assets/Biped/FrankenBruin/Animation/AimOffsets/FrankenBruin_AO
      BasePose:Pose Link Structure=@AnimGraphNode_StateMachine_0
      AlphaCurveName:Name=DisableAimOffset
     ) -> (Pose:Pose Link Structure) {alphaInputType=Curve, alphaBoolBlend {blendOption=Linear}, alphaScaleBiasClamp {bias=1, scale=-1}}
     Additive=Blendspace Player 'FrankenBruin_COR_Loco_Turn_BS'!(
      X=@n7
      Y=0
      BlendSpace:Blend Space Object Reference=/CosmeticCompanions/Assets/Biped/FrankenBruin/Animation/Core/FrankenBruin_COR_Loco_Turn_BS
     ) -> (Pose)
    ) {alphaBoolBlend {blendOption=Linear}}) {cacheName=PreVOPose}
   MapRangeClamped(Value=Get CompanionSpeed2D() -> (CompanionSpeed2D:float), InRangeA=-800, InRangeB=800, OutRangeA=10, OutRangeB=-15)
  ]}
 Locomotion {
  entry=Idle
  path=AnimGraph.AnimGraphNode_StateMachine_0.Locomotion
  ids=id
  nodes [
   AnimStateNode_0 = Idle()
   AnimStateNode_1 = Run() {is=title AnimStateNode}
   AnimStateNode_2 = Stop() {is=title AnimStateNode}
   AnimStateTransitionNode_0 = AnimStateTransitionNode() {bAllowInertializationForSelfTransitions=true}
   AnimStateTransitionNode_1 = AnimStateTransitionNode() {bAllowInertializationForSelfTransitions=true}
   AnimStateTransitionNode_2 = AnimStateTransitionNode() {bAllowInertializationForSelfTransitions=true, bAutomaticRuleBasedOnSequencePlayerInState=true}
  ]
  links [
   {
    dst=Run
    rule=AnimGraph.AnimGraphNode_StateMachine_0.Locomotion.AnimStateTransitionNode_0.Transition
    src=Idle
    via=AnimStateTransitionNode_0
   }
   {
    dst=Stop
    rule=AnimGraph.AnimGraphNode_StateMachine_0.Locomotion.AnimStateTransitionNode_1.Transition
    src=Run
    via=AnimStateTransitionNode_1
   }
   {
    dst=Idle
    rule=AnimGraph.AnimGraphNode_StateMachine_0.Locomotion.AnimStateTransitionNode_2.Transition
    src=Stop
    via=AnimStateTransitionNode_2
   }
  ]
 }
 Idle {path=AnimGraph.AnimGraphNode_StateMachine_0.Locomotion.AnimStateNode_0.Idle, nodes [Output Animation Pose()]}
 Run {
  path=AnimGraph.AnimGraphNode_StateMachine_0.Locomotion.AnimStateNode_1.Run
  nodes [
   n1 = "Blend Poses (ELocomotionMovementType)"!(
    BlendPose_0:Pose Link Structure=Sequence Player 'FrankenBruin_COR_Loco_Trot'!(
     Sequence:Anim Sequence Base Object Reference=/CosmeticCompanions/Assets/Biped/FrankenBruin/Animation/Core/FrankenBruin_COR_Loco_Trot
     PlayRate:float=1
    ) -> (Pose) {groupName=Locomotion, method=SyncGroup}
    BlendPose_1:Pose Link Structure=Sequence Player 'FrankenBruin_COR_Loco_Run'!(
     Sequence:Anim Sequence Base Object Reference=/CosmeticCompanions/Assets/Biped/FrankenBruin/Animation/Core/FrankenBruin_COR_Loco_Run
     PlayRate:float=1
    ) -> (Pose) {groupName=Locomotion, method=SyncGroup}
    BlendPose_2:Pose Link Structure=Sequence Player 'FrankenBruin_COR_Loco_Run'(Sequence=&FrankenBruin_COR_Loco_Run, PlayRate=1) {groupName=Locomotion, method=SyncGroup}
    BlendTime_0:float=0.1
    BlendTime_1:float=0.1
    BlendTime_2:float=0.1
    ActiveEnumValue:Byte=Get CurrentLocomotionMovementType() -> (CurrentLocomotionMovementType:ELocomotionMovementType Enum=Walk)
   ) -> (Pose:Pose Link Structure) {
    is=title AnimGraphNode_BlendListByEnum
    blendPose [{linkId=-1, sourceLinkId=-1}, {linkId=-1, sourceLinkId=-1}, {linkId=-1, sourceLinkId=-1}]
    blendTime [0.1,0.1,0.1]
   }
   Output Animation Pose(Result=@n1)
  ]
 }
 Stop {
  path=AnimGraph.AnimGraphNode_StateMachine_0.Locomotion.AnimStateNode_2.Stop
  nodes [
   AnimGraphNode_ChooserPlayer_0 = Chooser Player: CHT_Biped_Base_Stops!() -> (Pose:Pose Link Structure) {
    is=title AnimGraphNode_ChooserPlayer
    chooser {_structType=/Script/Chooser.EvaluateChooser, chooser=/CosmeticCompanions/Assets/Biped/BipedBase/Choosers/CHT_Biped_Base_Stops}
    chooserContextDefinition [
     {
      _structType=/Script/Chooser.ContextObjectTypeClass
      class=&/LaughPlier_AnimBP.LaughPlier_AnimBP_C
      direction=ReadWrite
     }
     {struct=/Script/Chooser.ChooserPlayerSettings, _structType=/Script/Chooser.ContextObjectTypeStruct, direction=Read}
    ]
   }
   Sequence Player 'FrankenBruin_COR_Loco_StopHeavy'() {bLoopAnimation=false, sequence=/CosmeticCompanions/Assets/Biped/FrankenBruin/Animation/Core/FrankenBruin_COR_Loco_StopHeavy}
   Output Animation Pose(Result=Sequence Player 'FrankenBruin_COR_Loco_StopLight'() {bLoopAnimation=false, sequence=/CosmeticCompanions/Assets/Biped/FrankenBruin/Animation/Core/FrankenBruin_COR_Loco_StopLight})
  ]
 }
 AnimationBlendStackGraph_0 {
  path=AnimGraph.AnimGraphNode_StateMachine_0.Locomotion.AnimStateNode_2.Stop.AnimGraphNode_ChooserPlayer_0.AnimationBlendStackGraph_0
  nodes [Output Pose!(Result:Pose Link Structure=Blend Stack Input!() -> (Pose:Pose Link Structure) {is=title AnimGraphNode_BlendStackInput}) {is=title AnimGraphNode_BlendStackResult}]
 }
 Transition {
  path=AnimGraph.AnimGraphNode_StateMachine_0.Locomotion.AnimStateTransitionNode_0.Transition
  nodes [Result(bCanEnterTransition=Get bIsMoving() -> (bIsMoving:bool))]
 }
 Transition {
  path=AnimGraph.AnimGraphNode_StateMachine_0.Locomotion.AnimStateTransitionNode_1.Transition
  nodes [Result(bCanEnterTransition=NOT Boolean(A=Get bIsMoving()))]
 }
 Transition {path=AnimGraph.AnimGraphNode_StateMachine_0.Locomotion.AnimStateTransitionNode_2.Transition, nodes [Result()]}
 EventGraph {nodes [
   n1 = Event BlueprintUpdateAnimation()
   > UpdateBallRoll!(execute, self, DeltaTime:double=@n1.DeltaTimeX) -> (then)
   > Set CompanionSpeedLastUpdate(CompanionSpeedLastUpdate:double=Get CompanionSpeed() -> (CompanionSpeed:double)) -> (Output_Get:double)
   n2 = GetOwningActor()
   > Set CompanionSpeed(CompanionSpeed:double=Vector Length XY!(A:Vector=GetVelocity!(self:Actor Object Reference=@n2) -> (ReturnValue:Vector)) -> (ReturnValue:double)) -> (Output_Get:double)
   > Set AccelerationLastUpdate(AccelerationLastUpdate:double=Get Acceleration() -> (Acceleration:double)) -> (Output_Get:double)
   n3 = Get CompanionSpeedLastUpdate() -> (CompanionSpeedLastUpdate:double)
   n4 = Get CompanionSpeed()
   > Set Acceleration(Acceleration:double=SafeDivide(A=float - float(A=@n4, B=@n3), B=0)) -> (Output_Get:double)
   n5 = Get AccelerationLastUpdate() -> (AccelerationLastUpdate:double)
   > Set LightStop(LightStop:bool=OR Boolean(A=Get bIsMoodWorried() -> (bIsMoodWorried:bool), B="float > float"(A=@n5, B=-1200))) -> (Output_Get:bool)
   n6 = Get CompanionSpeed2D()
   n7 = Set BallXRotation(BallXRotation:double=float + float(A=float * float(A=@n1.DeltaTimeX, B=@n6), B=Get BallXRotation() -> (BallXRotation:double))) -> (Output_Get:double)
   > n8 = Set BallXRotation(BallXRotation="Wrap (Float)"!(Value:double=@n7, Min:double=0, Max:double=360) -> (ReturnValue:double))
   n9 = Get CompanionDirection()
   > n10 = Set BallZRotation(BallZRotation:double=FInterpTo(Current=Get BallZRotation() -> (BallZRotation:double), Target=float * float(A=@n9, B=-1), DeltaTime=@n1.DeltaTimeX, InterpSpeed=3)) -> (Output_Get:double)
   > Set!(execute, BallRotation_Roll:float=@n8, BallRotation_Pitch:float=@n10, BallRotation_Yaw:float=0) -> (then, Output_Get:Rotator)
   TryGetPawnOwner()
  ]}
 UpdateBallRoll {nodes [
   n1 = UpdateBallRoll!() -> (then, DeltaTime:double=~) {is=title K2Node_FunctionEntry}
   n2 = float * float(A=@n1, B=Get CompanionSpeed2D())
   n3 = Get BallXRotation()
   > n4 = Set BallXRotation(BallXRotation="Wrap (Float)"(Value=float + float(A=@n2, B=@n3), Min=0, Max=360))
   n5 = Get CompanionDirection()
   > n6 = Set BallZRotation(BallZRotation=FInterpTo(Current=Get BallZRotation(), Target=float * float(A=@n5, B=-1), DeltaTime=Get DeltaTime() -> (DeltaTime:double), InterpSpeed=3))
   > Set(BallRotation_Roll=@n4, BallRotation_Pitch=@n6, BallRotation_Yaw=0)
   float + float(A=@n4, B=float * float(A=Get BallRollCurve() -> (BallRollCurve:double), B=20))
  ]}
]
variables [
 Acceleration {type=number}
 AccelerationLastUpdate {type=number}
 BallRollCurve {type=number}
 BallRotation {type=Rotator}
 BallXRotation {type=number}
 BallZRotation {type=number}
 CompanionSpeed {type=number}
 CompanionSpeedLastUpdate {type=number}
 LightStop {type=boolean}
]
