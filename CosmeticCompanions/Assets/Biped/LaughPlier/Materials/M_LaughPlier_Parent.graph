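# Parent material for the LaughPlier biped companion, blending a lit circuitry detail into base color, normal and emissive between the shared character PreFX and PostFX functions.
# An object-aligned virtual plane generates circuitry UVs from parameterized offset and axes.
# A fake light at a fixed UV position combines distance falloff with normal-map shading, tinted by LightColor.
# Tech-art, edge and view-based radial masks are built, alongside a temporal responsiveness output.
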Material /CosmeticCompanions/Assets/Biped/LaughPlier/Materials/M_LaughPlier_Parent
graph {
 properties {
  bAutomaticallySetUsageInEditor=false
  bHasPixelAnimation=true
  bUseMaterialAttributes=true
  bUsedWithMeshDeformer=true
  bUsedWithMorphTargets=true
  bUsedWithSkeletalMesh=true
  pixelDepthOffsetMode=PDOM_Legacy
  previewMesh=/CosmeticCompanions/Assets/Biped/LaughPlier/DEMO/LaughPlier
  shadingModel=MSM_FromMaterialExpression
  usedShadingModels=Default Lit | Subsurface | Clear Coat | Subsurface Profile | Cloth
 }
 nodes [
  n1 = MaterialFunctionCall() {materialFunction=&MF_CharacterParent_2023_PreFX}
  n2 = GetMaterialAttributes(None=@n1.Result) {attributeGetTypes [#1, #2, #3]}
  n3 = TextureSample(UVs=NamedRerouteUsage()) {
   samplerSource=SSM_Clamp_WorldGroupSettings
   samplerType=SAMPLERTYPE_Normal
   texture=/CosmeticCompanions/Assets/Biped/LaughPlier/Textures/T_LaughPlier_Circuitry_N
  }
  n4 = Multiply(A=NamedRerouteUsage(), B=NamedRerouteUsage()) {constB=0.05}
  n5 = NamedRerouteUsage()
  n6 = NamedRerouteDeclaration(None=ComponentMask(None=@n1.Tech Art Mask) {g=true}) {name=TechArtGMask, nodeColor=<rgba 0 1 0 0>}
  n7 = TextureSample(UVs=NamedRerouteUsage()) {texture=/CosmeticCompanions/Assets/Biped/LaughPlier/Textures/T_LaughPlier_Circuitry_E}
  n8 = Add(
   A=TextureSample(UVs=NamedRerouteUsage()) {texture=/CosmeticCompanions/Assets/Biped/LaughPlier/Textures/T_LaughPlier_Circuitry_BC}.RGB
   B=Multiply(A=Multiply(A=@n7.R) {constB=1.5}, B=VectorParameter() {parameterName=LightColor, defaultValue=<rgba 1 0 0 1>}.RGB)
  )
  n9 = Multiply(A=NamedRerouteUsage(), B=Multiply(A=NamedRerouteUsage(), B=NamedRerouteUsage()))
  MaterialFunctionCall(
   Material=SetMaterialAttributes!(
    MaterialAttributes=@n2.MaterialAttributes
    Normal=LinearInterpolate(A=@n2.Normal, B=@n3.RGB, Alpha=@n4) {desc=Normal}
    Emissive Color=LinearInterpolate(A=@n2.Emissive Color, B=@n5, Alpha=@n6) {desc=Emissive}
    Base Color=LinearInterpolate(A=@n2.Base Color, B=@n8, Alpha=@n9) {desc=BC}
   ) {attributeSetTypes [#1, #2, #3]}
   M=@n1.M
   Pre Skinned Local Position=@n1.M
   Distance Blend=@n1.M
   Fresnel=@n1.M
  ) {materialFunction=&MF_CharacterParent_2023_PostFX}
  n10 = ActorPositionWS()
  n11 = Transform(None=VectorParameter() {parameterName=WP Offset, defaultValue=<rgba -1.2143 15.5262 54.8649 0>}.RGB) {transformSourceType=TRANSFORMSOURCE_Local}
  n12 = Transform(None=VectorParameter() {parameterName=Normal Direction, defaultValue=<rgba 0 -1 -0.5 1>}.RGB) {transformSourceType=TRANSFORMSOURCE_Instance}
  n13 = Transform(None=VectorParameter() {parameterName=Tangent Direction, defaultValue=<rgba 1 0 0 1>}.RGB) {transformSourceType=TRANSFORMSOURCE_Instance}
  n14 = Transform(None=VectorParameter() {parameterName=Bitangent Direction, defaultValue=<rgba 0 0 1 1>}.RGB) {transformSourceType=TRANSFORMSOURCE_Instance}
  n15 = Subtract(
   A=MaterialFunctionCall!(
    Projection Pivot Position=Add(A=@n10, B=@n11)
    Plane Depth Offset=ScalarParameter() {parameterName=Param_1}
    "Coordinate Scale (WS)"=ScalarParameter() {defaultValue=8, parameterName="Coordinate Scale (WS) (S)"}
    World Space Projection Normal=Normalize(VectorInput=@n12)
    World Space Projection Tangent=Normalize(VectorInput=@n13)
    World Space Projection Binormal=Normalize(VectorInput=@n14)
   ) {materialFunction=/Engine/Functions/Engine_MaterialFunctions02/UVs/ObjectAlignedVirtualPlaneCoordinates}.Center Aligned UV Plane
   B=ComponentMask(None=VectorParameter() {parameterName=Offset UV, defaultValue=<rgba 0.1664 0.7624 0.192 0>}.RGB) {g=true, r=true}
  ) {constB=0.5}
  n16 = ComponentMask(None=@n15) {r=true}
  n17 = OneMinus(None=ComponentMask(None=@n15) {g=true})
  NamedRerouteDeclaration(None=Saturate(None=AppendVector(A=@n16, B=@n17))) {name=AlignedUVs, nodeColor=<rgba 1 0 0 0>}
  n18 = NamedRerouteUsage()
  NamedRerouteDeclaration(None=Multiply(A=TextureSample(UVs=@n18) {texture=/Game/Global/Textures/Default/Masks_and_Gradients/WM_PinGlow_Sphere}.R) {constB=0.5}) {name=EdgeCircuitryMask, nodeColor=<rgba 1 0.7734 0 1>}
  n19 = MaterialFunctionCall!(
   value=Distance(A=NamedRerouteUsage(), B=NamedRerouteUsage())
   scale=ScalarParameter() {defaultValue=0.12, parameterName=LightFalloff_Size}
   invert value=StaticBool()
   pow=ScalarParameter() {defaultValue=1.75, parameterName=LightFalloff_Exponent}
  ) {materialFunction=/Game/Global/Materials/MaterialGym/Materials/MaterialElements/GrayscaleTransforms/InverseProportional/MF_InverseProportional}
  n20 = Subtract(
   A=NamedRerouteUsage()
   B=NamedRerouteDeclaration(None=Constant2Vector() {desc=Light position in UV space, g=0.55, r=0.32}) {name=LightPosition, nodeColor=<rgba 0.9688 0 1 1>}
  )
  n21 = AppendVector(A=Multiply(A=Normalize(VectorInput=@n20)) {constB=-1}, B=Constant() {r=0.25})
  n22 = SphereMask(A=NamedRerouteUsage(), B=NamedRerouteUsage()) {attenuationRadius=0.6, hardnessPercent=0}
  n23 = Normalize(VectorInput=LinearInterpolate(A=Constant3Vector() {constant=<rgba 0 0 -1 0>}, B=Normalize(VectorInput=@n21), Alpha=Power(Base=@n22)))
  n24 = TextureSample(UVs=NamedRerouteUsage()) {
   samplerSource=SSM_Clamp_WorldGroupSettings
   samplerType=SAMPLERTYPE_Normal
   texture=&T_LaughPlier_Circuitry_N
  }
  n25 = Add(A=@n19.Result, B=Saturate(None=DotProduct(A=@n23, B=@n24.RGB)))
  n26 = SphereMask(A=TextureCoordinate(), B=Constant2Vector() {g=0.51, r=0.125}) {attenuationRadius=0.08, hardnessPercent=0}
  n27 = Multiply(A=NamedRerouteUsage(), B=Multiply(A=Power(Base=@n26)) {constB=0.5})
  NamedRerouteDeclaration(None=Multiply(A=Add(A=@n25, B=@n27), B=VectorParameter() {parameterName=LightColor, defaultValue=<rgba 1 0 0 1>}.RGB)) {name=NewEmissive, nodeColor=<rgba 0.5469 0 1 1>}
  n28 = Power(Base=DotProduct(A=PixelNormalWS(), B=CameraVectorWS())) {desc=Radial coordinate distortion}
  n29 = Subtract(A=PixelNormalWS(), B=CameraVectorWS())
  n30 = Normalize(VectorInput=Multiply(
    A=Transform(None=@n29) {transformSourceType=TRANSFORMSOURCE_World, transformType=TRANSFORM_Camera}
    B=Constant3Vector() {constant=<rgba 1 1 0 0>}
   ))
  n31 = Constant3Vector() {constant=<rgba 0 1 0 0>}
  n32 = ComponentMask(None=CrossProduct(A=@n30, B=@n31)) {b=true}
  n33 = Abs(None=@n32)
  n34 = MaterialFunctionCall(In=@n33, Power=Constant() {r=0.75}) {
   desc="Kind of optional, but smoothes out\r\nsome nonlinear distortion"
   materialFunction=&SCurve
  }
  n35 = Multiply(A=QualitySwitch(Default=@n34.Result, Low=@n33, Medium=@n33), B=Sign(None=@n32))
  n36 = DotProduct(A=@n30, B=@n31)
  n37 = LinearInterpolate(A=ConstantBiasScale(None=@n35) {bias=-3, scale=-0.25}, B=ConstantBiasScale(None=@n35) {scale=0.25}, Alpha=Saturate(None=Sign(None=@n36))) {constB=0}
  NamedRerouteDeclaration(None=ComponentMask(None=AppendVector(A=@n28, B=@n37) {desc=Output UVs}) {r=true}) {name=RadialMask, nodeColor=<rgba 0.7578 0 1 1>}
  TemporalResponsivenessOutput!(None=LinearInterpolate(Alpha=NamedRerouteUsage()) {constB=0.5})
 ]
}
